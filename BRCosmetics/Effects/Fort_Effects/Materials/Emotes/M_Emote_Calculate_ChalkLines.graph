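# Additive unlit mesh-particle material for the chalk lines of the Calculate emote.
# Tiled cloud noise, softened by a parabolic V-center mask, is tinted by particle color and passed through the shared emissive adjustment function.
# A second chain combines noise, center mask, trail-mask texture, particle alpha and inverted vertex red.
# Dynamic parameters Fill and FadeOutPanner drive that chain's panned gradients, one revealing the line, the other rotated and inverted to fade it.
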Material /BRCosmetics/Effects/Fort_Effects/Materials/Emotes/M_Emote_Calculate_ChalkLines
graph {
 properties {
  bUsedWithMeshParticles=true
  bUsedWithNiagaraMeshParticles=true
  blendMode=BLEND_Additive
  pixelDepthOffsetMode=PDOM_Legacy
  shadingModel=MSM_Unlit
  usedShadingModels=Unlit
 }
 nodes [
  n1 = TextureCoordinate()
  n2 = AppendVector(A=ScalarParameter() {defaultValue=30, parameterName=MultXTile}, B=ScalarParameter() {defaultValue=1, parameterName=MultYTile})
  n3 = LinearInterpolate(
   A=ScalarParameter() {parameterName=A}
   B=ScalarParameter() {defaultValue=1, parameterName=B}
   Alpha=TextureSample(UVs=Multiply(A=@n1, B=@n2)) {texture=&AtmosphericCloudNoise01}.G
  )
  n4 = Constant() {r=1}
  n5 = ComponentMask(None=TextureCoordinate()) {g=true}
  n6 = Multiply(A=Constant() {r=2}, B=Multiply(A=@n5, B=OneMinus(None=@n5)))
  n7 = ParticleColor()
  MaterialFunctionCall(Emissive=Multiply(A=LinearInterpolate(A=@n3, B=@n4, Alpha=@n6), B=@n7.RGB)) {materialFunction=&MF_FX_EmissiveAdjustments}
  n8 = MaterialFunctionCall(
   UVs=Panner(
    Coordinate=TextureCoordinate() {uTiling=0.35, vTiling=0.35}
    Time=DynamicParameter() {defaultValue=<rgba 0.5 0.35 1 1>, paramNames [Fill, FadeOutPanner, Param3, Param4]}.FadeOutPanner
   ) {speedX=1}
   "Rotation Angle (0-1)"=ScalarParameter() {defaultValue=0.35, parameterName="Rotation Angle (0-1) (S)"}
  ) {materialFunction=&CustomRotator}
  n9 = TextureSample(UVs=Panner(
    Coordinate=TextureCoordinate() {uTiling=0.45}
    Time=DynamicParameter() {defaultValue=<rgba 0.5 0.35 1 1>, paramNames [Fill, FadeOutPanner, Param3, Param4]}.Fill
   ) {speedX=1}) {texture=/Game/Tools/Turntable/Textures/T_ThinkBase_Teleporter_Gradient}
  n10 = ScalarParameter() {defaultValue=3, parameterName=MultXTile_1}
  n11 = ScalarParameter() {defaultValue=1, parameterName=MultYTile_1}
  n12 = TextureSample(UVs=Multiply(A=TextureCoordinate(), B=AppendVector(A=@n10, B=@n11))) {texture=/Game/Effects/Fort_Effects/Textures/Smoke/T_Lobber_Trail_Mask}
  n13 = Multiply(A=Multiply(A=Multiply(A=@n7.A, B=@n9.G), B=@n6), B=Power(Base=Multiply(A=@n12.G, B=@n12.G)) {constExponent=0.25})
  n14 = VertexColor()
  Multiply(
   A=OneMinus(None=TextureSample(UVs=@n8.Rotated Values) {texture=&T_ThinkBase_Teleporter_Gradient}.RGB)
   B=Multiply(A=Multiply(A=@n3, B=@n13), B=OneMinus(None=@n14.R))
  )
 ]
}
