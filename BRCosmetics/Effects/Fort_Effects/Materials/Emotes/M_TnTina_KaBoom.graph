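# Masked, unlit, two-sided material for Niagara mesh particles that draws the TnTina "KaBoom" emote effect, using a tinted SubUV texture for emissive.
# Dynamic parameters drive a grow/shrink vertex offset about an adjustable center, with vertex-color green split at a threshold to choose Ka or Boom.
# The offset adds camera offset and a sine wiggle phased by vertex red.
# Optional camera-proximity shrinking and a debug sine animation are included.
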Material /BRCosmetics/Effects/Fort_Effects/Materials/Emotes/M_TnTina_KaBoom
graph {
 properties {
  bUsedWithNiagaraMeshParticles=true
  blendMode=BLEND_Masked
  pixelDepthOffsetMode=PDOM_Legacy
  shadingModel=MSM_Unlit
  twoSided=true
  usedShadingModels=Unlit
 }
 nodes [
  n1 = Add(A=Subtract(A=WorldPosition().XYZ, B=ObjectPositionWS()), B=VectorParameter() {parameterName=Center Offset, defaultValue=<rgba 0 0 80 0>}.RGB)
  n2 = VertexColor()
  n3 = Constant() {r=0.785}
  n4 = DynamicParameter() {defaultValue=<rgba 1 1 1 0>, paramNames [GrowShrink,Ka_Or_Boom,WigglePower,CameraOffset]}
  n5 = MaterialFunctionCall() {materialFunction=&DebugTimeSine}
  n6 = OneMinus(None=LinearInterpolate(A=@n4.GrowShrink, B=Multiply(A=@n5.Result) {constB=2}, Alpha=ScalarParameter() {parameterName=DEBUG}))
  n7 = Subtract(A=@n2.G, B=@n6)
  n8 = If(A=@n2.G, B=@n3, "A > B"=@n7, "A == B"=@n7, "A < B"=Constant())
  n9 = If(A=@n2.G, B=@n3, "A > B"=Constant(), "A == B"=@n7, "A < B"=@n7)
  n10 = LinearInterpolate(
   A=Clamp(None=LinearInterpolate(Alpha=@n8) {constA=-1}) {minDefault=-1}
   B=Clamp(None=LinearInterpolate(Alpha=@n9) {constA=-1}) {minDefault=-1}
   Alpha=@n4.Ka_Or_Boom
  )
  n11 = LinearInterpolate(
   B=MaterialFunctionCall(Offset Amount=DynamicParameter() {defaultValue=<rgba 1 1 1 0>, paramNames [GrowShrink,Ka_Or_Boom,WigglePower,CameraOffset]}.CameraOffset) {materialFunction=/Engine/Functions/Engine_MaterialFunctions02/WorldPositionOffset/CameraOffset}.Camera Offset
   Alpha=Clamp(None=@n10)
  )
  n12 = Multiply(A=Add(A=Time(), B=VertexColor().R), B=ScalarParameter() {defaultValue=3, parameterName=WiggleSpeed}) {constB=1.5}
  n13 = Multiply(A=ConstantBiasScale(None=Sine(None=@n12)) {bias=1.5}, B=ScalarParameter() {defaultValue=5, parameterName=Wiggle Strength})
  n14 = Transform(None=MaterialFunctionCall(G=@n13, B=@n13) {materialFunction=&AppendMany}.RGB) {transformSourceType=TRANSFORMSOURCE_Local}
  n15 = Power(Base=@n6, Exp=@n4.WigglePower) {constExponent=0.5}
  n16 = Multiply(A=Add(A=Add(A=@n1, B=@n11), B=Multiply(A=@n14, B=@n15)), B=@n10) {constA=1}
  n17 = ParticlePositionWS()
  n18 = Subtract(A=@n17, B=WorldPosition().XYZ)
  n19 = Subtract(
   A=Distance(A=@n17, B=CameraPositionWS())
   B=ScalarParameter() {defaultValue=200, group="8) WPO", parameterName=Sprite Shrink Camera Distance, sortPriority=80}
  )
  n20 = ScalarParameter() {defaultValue=150, group="8) WPO", parameterName=Sprite Shrink Distance Falloff Hardness, sortPriority=81}
  n21 = OneMinus(None=Saturate(None=Divide(A=@n19, B=@n20)))
  LinearInterpolate(A=@n16, B=Add(A=@n16, B=Multiply(A=@n18, B=@n21)), Alpha=ScalarParameter() {parameterName=ScaleSpriteByCameraDistance?})
  n22 = TextureSampleParameter2D() {parameterName=Param, texture=/BRCosmetics/Athena/Effects/Textures/Emotes/T_TnTina_Kaboom_SubUV}
  n23 = Multiply(A=@n22.G, B=VectorParameter() {parameterName=Color Multiplier, defaultValue=<rgba 1 0.5718 0 0>}.RGB)
  MaterialFunctionCall(Emissive=@n23) {materialFunction=&MF_FX_EmissiveAdjustments}
  Reroute(None=@n22.B)
  Reroute(None=@n23)
 ]
}
